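# Physics asset for the medium male Bronto backpack cosmetic, defining collision bodies and joints for its dynamic tail.
# Four tail boxes, mostly shrinking toward the tip, chain from a tiny dyn_tail_MSTR box through constraints with identical moderate swing and tight twist limits.
# The root body combines a capsule and a box, and no constraint links it to the tail chain.
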PhysicsAsset /BRCosmetics/Accessories/FORT_Backpacks/M_MED_Bronto_Backpack/M_MED_Bronto_Pack_PhysicsAsset
bodies {
 dyn_tail_01 [{shapeType=Box, extentX=14, extentY=12, extentZ=10.5, center=<1 0 0>, rotation=<rot 90 0 0>}]
 dyn_tail_02 [{shapeType=Box, extentX=12, extentY=11.5, extentZ=9.5, center=<2 0 0>, rotation=<rot 90 0 0>}]
 dyn_tail_03 [{shapeType=Box, extentX=10.5, extentY=8, extentZ=9.5, center=<4 0 0>, rotation=<rot 90 0 0>}]
 dyn_tail_04 [{shapeType=Box, extentX=6.5, extentY=7.5, extentZ=9.5, center=<8 0 0>, rotation=<rot 90 0 0>}]
 dyn_tail_MSTR [{shapeType=Box, extentX=1.01, extentY=1.01, extentZ=1.01, rotation=<rot 0 -90 0>}]
 root [
  {shapeType=Capsule, length=15, radius=15, center=<0.0008 12.0721 -33.8625>, rotation=<rot 90 -0.0099 -0.0099>}
  {shapeType=Box, extentX=41.5, extentY=5, extentZ=28.6477, center=<0.0001 -0.7714 -4.8369>, rotation=<rot 0 0 -0.0012>}
 ]
}
constraints [
 {bone1Name=dyn_tail_01, bone2Name=dyn_tail_MSTR, swing1LimitDegrees=35, swing2LimitDegrees=45, twistLimitDegrees=5}
 {bone1Name=dyn_tail_02, bone2Name=dyn_tail_01, swing1LimitDegrees=35, swing2LimitDegrees=45, twistLimitDegrees=5}
 {bone1Name=dyn_tail_03, bone2Name=dyn_tail_02, swing1LimitDegrees=35, swing2LimitDegrees=45, twistLimitDegrees=5}
 {bone1Name=dyn_tail_04, bone2Name=dyn_tail_03, swing1LimitDegrees=35, swing2LimitDegrees=45, twistLimitDegrees=5}
]
