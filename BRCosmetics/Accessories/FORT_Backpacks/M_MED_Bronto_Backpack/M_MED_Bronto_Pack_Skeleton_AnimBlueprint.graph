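# Animation Blueprint for the Bronto backpack cosmetic, adding secondary physics motion and an offset during skydive floating.
# Each update converts bIsSkydiveFloating to a number in is_floating, which serves as the alpha for an additive bone-space shift and roll on the root bone.
# The adjusted input pose then runs through a RigidBody node simulating in pack-bone space with the pack physics asset, overridden gravity and scaled, clamped component motion.
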AnimBlueprint /BRCosmetics/Accessories/FORT_Backpacks/M_MED_Bronto_Backpack/M_MED_Bronto_Pack_Skeleton_AnimBlueprint
target_skeleton=&/M_MED_Bronto_Pack_Skeleton
graphs [
 AnimGraph {nodes [
   AnimGraphNode_SubInput_1 = Input Pose()
   n1 = "Transform (Modify) Bone - Bone: root"(
    ComponentPose=Local To Component(LocalPose=@AnimGraphNode_SubInput_1)
    Alpha=Get is_floating() -> (is_floating:double)
    Translation="0.000,-2.3,0.000"
    Rotation="0.000,0.000,15.000"
   ) {
    rotationMode=BMM_Additive
    rotationSpace=BCS_BoneSpace
    translationMode=BMM_Additive
    translationSpace=BCS_BoneSpace
    alphaBoolBlend {blendOption=Linear}
    boneToModify {boneName=root}
    rotation {roll=15.0004}
    translation {y=-2.3}
   }
   n2 = RigidBody(ComponentPose=@n1) {
    bClampLinearTranslationLimitToRefPose=true
    bOverrideWorldGravity=true
    overridePhysicsAsset=&/M_MED_Bronto_Pack_PhysicsAsset
    simulationSpace=BaseBoneSpace
    alphaBoolBlend {blendOption=Linear}
    baseBoneRef {boneName=pack}
    componentAppliedLinearAccClamp=<3000 2000 2000>
    componentLinearAccScale=<0.5 0.3 0.3>
    componentLinearVelScale=<0.5 0.3 0.3>
    overrideWorldGravity {z=-980}
   }
   Output Pose(Result=Component To Local(ComponentPose=@n2)) {layerGroup=None}
  ]}
 EventGraph {nodes [
   Event BlueprintUpdateAnimation()
   > Set is_floating(is_floating:double="To Float (Boolean)"(InBool=Get bIsSkydiveFloating())) -> (Output_Get:double)
   TryGetPawnOwner()
  ]}
]
variables [is_floating {type=number}]
